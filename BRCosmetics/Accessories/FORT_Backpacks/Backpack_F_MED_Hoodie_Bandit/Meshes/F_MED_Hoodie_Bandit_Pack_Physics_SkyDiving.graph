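# Physics asset for the Hoodie Bandit backpack cosmetic, by its name the skydiving variant, covering its dangling bags, holster, zippers, top straps and ribbons.
# Master bodies hang off the central backpack body, mostly with free angular joints, and each carries a limit-constrained child.
# Two five-link ribbon chains bend within swing limits, the right attaching directly to the backpack.
# Both top straps have duplicate constraints to their masters, one reversed on the right.
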PhysicsAsset /BRCosmetics/Accessories/FORT_Backpacks/Backpack_F_MED_Hoodie_Bandit/Meshes/F_MED_Hoodie_Bandit_Pack_Physics_SkyDiving
bodies {
 dyn_bag_l [{shapeType=Box, extentX=1.9701, extentY=5.4869, extentZ=6.2148, center=<0.6257 0.0339 0.1724>}]
 dyn_bag_r [{shapeType=Box, extentX=1.6859, extentY=5.4914, extentZ=6.1798, center=<0.5936 0.018 0.2294>}]
 dyn_holster [{shapeType=Box, extentX=5.3499, extentY=12.852, extentZ=20.5036, center=<2.6104 -1.8937 -2.9615>}]
 dyn_mstr_backpack [
  {shapeType=Capsule, length=0.505, radius=0.505}
  {shapeType=Capsule, length=5, radius=4.75, center=<-4.7831 0 5.5251>, rotation=<rot 0 0 89.9997>}
  {shapeType=Capsule, length=8, radius=2.9045, center=<1.41 -0.0002 -9.4669>, rotation=<rot 0 0 -89.999>}
  {shapeType=Capsule, length=7.9955, radius=7.5179, shapeName=dyn_mstr_backpack_capsule, center=<-0.2069 0 0>}
  {shapeType=Box, extentX=4.6996, extentY=25.5, extentZ=21.5, center=<-1.6715 0 -17.3578>, rotation=<rot 6.6717 0 0>}
 ]
 dyn_mstr_bag_l [{shapeType=Capsule, length=0.505, radius=0.505}]
 dyn_mstr_bag_r [{shapeType=Capsule, length=0.505, radius=0.505}]
 dyn_mstr_holster [{shapeType=Capsule, length=6.25, radius=4.5, center=<-4.7121 0 -4.4735>}]
 dyn_mstr_ribbon_l [{shapeType=Capsule, length=2.1064, radius=0.5014, rotation=<rot -3.4633 0.0001 0>}]
 dyn_mstr_topStrap1_l [{shapeType=Capsule, length=5, radius=1.75, center=<2.4514 1.9135 -0.1401>, rotation=<rot -86.6629 -0.0041 -14.3163>}]
 dyn_mstr_topStrap1_r [{shapeType=Capsule, length=5, radius=1.75, center=<2.6364 1.5992 0.0339>, rotation=<rot -84.8163 0.8217 -13.3961>}]
 dyn_mstr_zipper_l [{shapeType=Capsule, length=0.505, radius=0.6704}]
 dyn_mstr_zipper_r [{shapeType=Capsule, length=0.3846, radius=0.6704, center=<0.0498 -0.0036 -0.061>, rotation=<rot 0.0002 -4.0824 0.0001>}]
 dyn_ribbon1_l [{shapeType=Box, extentX=2.5646, extentY=1.01, extentZ=3.0831, center=<0.1509 -0.4296 -0.0037>}]
 dyn_ribbon1_r [{shapeType=Box, extentX=2.6573, extentY=1.0164, extentZ=3.0831, center=<-0.1171 -0.3231 0.0034>}]
 dyn_ribbon2_l [{shapeType=Box, extentX=3.4127, extentY=1.032, extentZ=2.5, center=<1.0561 -0.129 0>}]
 dyn_ribbon2_r [{shapeType=Box, extentX=3.4127, extentY=1.032, extentZ=2.5, center=<1.0561 -0.0181 0>}]
 dyn_ribbon3_l [{shapeType=Capsule, length=2.0289, radius=1.3796, shapeName=dyn_ribbon3_l_capsule, center=<1.3712 0 0>, rotation=<rot 90 0 0>}]
 dyn_ribbon3_r [{shapeType=Capsule, length=2.0289, radius=1.4066, shapeName=dyn_ribbon3_r_capsule, center=<1.3712 0 0>, rotation=<rot 90 0 0>}]
 dyn_ribbon4_l [{shapeType=Capsule, length=2.4673, radius=1.4491, shapeName=dyn_ribbon4_l_capsule, center=<1.0008 0 0>, rotation=<rot 90 0 0>}]
 dyn_ribbon4_r [{shapeType=Capsule, length=2.4673, radius=1.4965, shapeName=dyn_ribbon4_r_capsule, center=<1.0008 0 0>, rotation=<rot 90 0 0>}]
 dyn_ribbon5_l [{shapeType=Capsule, length=1.6702, radius=1.622, shapeName=dyn_ribbon5_l_capsule, center=<1.0008 0 0>, rotation=<rot 90 0 0>}]
 dyn_ribbon5_r [{shapeType=Capsule, length=1.9888, radius=1.7271, shapeName=dyn_ribbon5_r_capsule, center=<1.2347 0 0>, rotation=<rot 90 0 0>}]
 dyn_topStrap1_l [
  {shapeType=Box, extentX=3.2594, extentY=0.8342, extentZ=4.574, center=<2.1344 -0.1925 -0.0033>, rotation=<rot -0.193 -11.4992 -2.0494>}
  {
   shapeType=Box
   extentX=2.0444
   extentY=0.9559
   extentZ=4.9821
   shapeName=dyn_topStrap1_l_box
   center=<-0.5977 0.1303 -0.0025>
   rotation=<rot 0 -4.2691 -0.1578>
  }
 ]
 dyn_topStrap1_r [
  {shapeType=Box, extentX=2.1594, extentY=0.9093, extentZ=4, center=<-1.0061 0.5263 0>, rotation=<rot 0 -27.9449 0>}
  {
   shapeType=Box
   extentX=3.4056
   extentY=0.9391
   extentZ=4.0951
   shapeName=dyn_topStrap1_r_box
   center=<1.7862 -0.2939 0>
   rotation=<rot 0 -12.2433 0>
  }
 ]
 dyn_zipper_l [{
   shapeType=Capsule
   length=1.703
   radius=0.7652
   shapeName=dyn_zipper_l_capsule
   center=<1.0637 -0.2301 0.3438>
   rotation=<rot 0 89.9162 -87.9527>
  }]
 dyn_zipper_r [{
   shapeType=Capsule
   length=1.726
   radius=0.7652
   shapeName=dyn_zipper_r_capsule
   center=<1.0618 0.1977 0.219>
   rotation=<rot 0 -91.3546 -89.3845>
  }]
}
constraints [
 {
  bone1Name=dyn_mstr_zipper_l
  bone2Name=dyn_mstr_backpack
  swing1LimitDegrees=45
  swing1Motion=Free
  swing2LimitDegrees=45
  swing2Motion=Free
  twistLimitDegrees=45
  twistMotion=Free
 }
 {bone1Name=dyn_zipper_l, bone2Name=dyn_mstr_zipper_l, swing1LimitDegrees=25, swing2LimitDegrees=25, twistLimitDegrees=10}
 {
  bone1Name=dyn_mstr_zipper_r
  bone2Name=dyn_mstr_backpack
  swing1LimitDegrees=45
  swing1Motion=Free
  swing2LimitDegrees=45
  swing2Motion=Free
  twistLimitDegrees=45
  twistMotion=Free
 }
 {bone1Name=dyn_zipper_r, bone2Name=dyn_mstr_zipper_r, swing1LimitDegrees=25, swing2LimitDegrees=25, twistLimitDegrees=10}
 {bone1Name=dyn_holster, bone2Name=dyn_mstr_holster, swing1LimitDegrees=5, swing2LimitDegrees=15, twistLimitDegrees=3}
 {
  bone1Name=dyn_mstr_holster
  bone2Name=dyn_mstr_backpack
  swing1LimitDegrees=45
  swing1Motion=Free
  swing2LimitDegrees=45
  swing2Motion=Free
  twistLimitDegrees=45
  twistMotion=Free
 }
 {
  bone1Name=dyn_ribbon1_r
  bone2Name=dyn_mstr_backpack
  swing1LimitDegrees=45
  swing1Motion=Free
  swing2LimitDegrees=45
  swing2Motion=Free
  twistLimitDegrees=45
  twistMotion=Free
 }
 {bone1Name=dyn_ribbon2_r, bone2Name=dyn_ribbon1_r, swing1LimitDegrees=45, swing2LimitDegrees=30.6, twistLimitDegrees=5, twistMotion=Locked}
 {bone1Name=dyn_ribbon3_r, bone2Name=dyn_ribbon2_r, swing1LimitDegrees=35, swing2LimitDegrees=35, twistLimitDegrees=10}
 {bone1Name=dyn_ribbon4_r, bone2Name=dyn_ribbon3_r, swing1LimitDegrees=35, swing2LimitDegrees=35, twistLimitDegrees=15}
 {bone1Name=dyn_ribbon5_r, bone2Name=dyn_ribbon4_r, swing1LimitDegrees=35, swing2LimitDegrees=35, twistLimitDegrees=15}
 {
  bone1Name=dyn_mstr_ribbon_l
  bone2Name=dyn_mstr_backpack
  swing1LimitDegrees=45
  swing1Motion=Free
  swing2LimitDegrees=45
  swing2Motion=Free
  twistLimitDegrees=45
  twistMotion=Free
 }
 {
  bone1Name=dyn_ribbon1_l
  bone2Name=dyn_mstr_ribbon_l
  swing1LimitDegrees=45
  swing1Motion=Free
  swing2LimitDegrees=45
  swing2Motion=Free
  twistLimitDegrees=45
  twistMotion=Free
 }
 {bone1Name=dyn_ribbon2_l, bone2Name=dyn_ribbon1_l, swing1LimitDegrees=45, swing2LimitDegrees=30.6, twistLimitDegrees=5, twistMotion=Locked}
 {bone1Name=dyn_ribbon3_l, bone2Name=dyn_ribbon2_l, swing1LimitDegrees=35, swing2LimitDegrees=35, twistLimitDegrees=10}
 {bone1Name=dyn_ribbon4_l, bone2Name=dyn_ribbon3_l, swing1LimitDegrees=35, swing2LimitDegrees=35, twistLimitDegrees=15}
 {bone1Name=dyn_ribbon5_l, bone2Name=dyn_ribbon4_l, swing1LimitDegrees=35, swing2LimitDegrees=35, twistLimitDegrees=5}
 {
  bone1Name=dyn_mstr_bag_r
  bone2Name=dyn_mstr_backpack
  swing1LimitDegrees=45
  swing1Motion=Free
  swing2LimitDegrees=45
  swing2Motion=Free
  twistLimitDegrees=45
  twistMotion=Free
 }
 {bone1Name=dyn_bag_r, bone2Name=dyn_mstr_bag_r, swing1LimitDegrees=7, swing2LimitDegrees=13, twistLimitDegrees=1}
 {
  bone1Name=dyn_mstr_bag_l
  bone2Name=dyn_mstr_backpack
  swing1LimitDegrees=45
  swing1Motion=Free
  swing2LimitDegrees=45
  swing2Motion=Free
  twistLimitDegrees=45
  twistMotion=Free
 }
 {bone1Name=dyn_bag_l, bone2Name=dyn_mstr_bag_l, swing1LimitDegrees=5, swing2LimitDegrees=10, twistLimitDegrees=1}
 {
  bone1Name=dyn_mstr_topStrap1_l
  bone2Name=dyn_mstr_backpack
  swing1LimitDegrees=15
  swing2LimitDegrees=15
  twistLimitDegrees=45
  twistMotion=Locked
 }
 {bone1Name=dyn_topStrap1_l, bone2Name=dyn_mstr_topStrap1_l, swing1LimitDegrees=15, swing2LimitDegrees=15, twistLimitDegrees=5}
 {
  bone1Name=dyn_mstr_topStrap1_r
  bone2Name=dyn_mstr_backpack
  swing1LimitDegrees=45
  swing1Motion=Free
  swing2LimitDegrees=45
  swing2Motion=Free
  twistLimitDegrees=45
  twistMotion=Free
 }
 {bone1Name=dyn_topStrap1_r, bone2Name=dyn_mstr_topStrap1_r, swing1LimitDegrees=15, swing2LimitDegrees=15, twistLimitDegrees=5}
 {
  bone1Name=dyn_topStrap1_r
  bone2Name=dyn_mstr_topStrap1_r
  swing1LimitDegrees=15
  swing2LimitDegrees=15
  twistLimitDegrees=45
  twistMotion=Locked
 }
 {
  bone1Name=dyn_mstr_topStrap1_r
  bone2Name=dyn_topStrap1_r
  swing1LimitDegrees=15
  swing2LimitDegrees=15
  twistLimitDegrees=5
  twistMotion=Locked
 }
 {
  bone1Name=dyn_topStrap1_l
  bone2Name=dyn_mstr_topStrap1_l
  swing1LimitDegrees=15
  swing2LimitDegrees=15
  twistLimitDegrees=45
  twistMotion=Locked
 }
]
