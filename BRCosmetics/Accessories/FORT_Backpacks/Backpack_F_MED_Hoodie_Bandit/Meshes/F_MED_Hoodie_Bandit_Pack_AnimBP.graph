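# Animation Blueprint for the Bandit hoodie backpack, adding physics motion to its flap and straps with separate setups for normal play and skydiving.
# Four chained AnimDynamics nodes swing the flap bones on angular springs, feeding two RigidBody stages with separate physics assets, each followed by small ribbon bone corrections.
# Each update turns bIsSkyDiving into opposite 0/1 alphas, so only the normal or skydiving RigidBody stage and its corrections apply.
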AnimBlueprint /BRCosmetics/Accessories/FORT_Backpacks/Backpack_F_MED_Hoodie_Bandit/Meshes/F_MED_Hoodie_Bandit_Pack_AnimBP
target_skeleton=&/F_MED_Hoodie_Bandit_Pack_Skeleton
graphs [
 AnimGraph {nodes [
   n1 = Anim Dynamics - Bone: dyn_flap1(ComponentPose=Local To Component(LocalPose=Local Space Ref Pose()), Alpha=1, ExternalForce="0,0,0") {
    angularDampingOverride=1
    angularSpringConstant=20000
    bAngularSpring=true
    bGravityOverrideInSimSpace=true
    bOverrideAngularDamping=true
    bUseGravityOverride=true
    simulationSpace=RootRelative
    alphaBoolBlend {blendOption=Linear}
    boundBone {boneName=dyn_flap1}
    componentAppliedLinearAccClamp=<3000 3000 3000>
    componentLinearAccScale=<0.3 0.3 0.3>
    componentLinearVelScale=<0.3 0.3 0.3>
    gravityOverride {z=-980}
    physicsBodyDefinitions [{
      sphereCollisionRadius=0
      boundBone {boneName=dyn_flap1}
      boxExtents=<5 5 5>
      constraintSetup {angularTargetAxis=AxisY, angularLimitsMax=<0 5 0>, angularLimitsMin=<0 0 0>, angularTarget=<0 1 0>}
      localJointOffset=<1 0 0.5>
     }]
   }
   n2 = Anim Dynamics - Bone: dyn_flap2(ComponentPose=@n1, Alpha=1, ExternalForce="0,0,0") {
    angularDampingOverride=0.97
    angularSpringConstant=20000
    bAngularSpring=true
    bGravityOverrideInSimSpace=true
    bOverrideAngularDamping=true
    bUseGravityOverride=true
    simulationSpace=RootRelative
    alphaBoolBlend {blendOption=Linear}
    boundBone {boneName=dyn_flap2}
    componentAppliedLinearAccClamp=<3000 3000 3000>
    componentLinearAccScale=<0.3 0.3 0.3>
    componentLinearVelScale=<0.3 0.3 0.3>
    gravityOverride {z=-980}
    physicsBodyDefinitions [{
      sphereCollisionRadius=0
      boundBone {boneName=dyn_flap2}
      boxExtents=<5 5 5>
      constraintSetup {angularTargetAxis=AxisY, angularLimitsMax=<0 7.5 0>, angularLimitsMin=<0 0 0>, angularTarget=<0 1 0>}
      localJointOffset=<1 0 0.5>
     }]
   }
   n3 = Anim Dynamics - Bone: dyn_flap3(ComponentPose=@n2, Alpha=1, ExternalForce="0,0,0") {
    angularDampingOverride=0.97
    angularSpringConstant=20000
    bAngularSpring=true
    bGravityOverrideInSimSpace=true
    bOverrideAngularDamping=true
    bUseGravityOverride=true
    simulationSpace=RootRelative
    alphaBoolBlend {blendOption=Linear}
    boundBone {boneName=dyn_flap3}
    componentAppliedLinearAccClamp=<3000 3000 3000>
    componentLinearAccScale=<0.3 0.3 0.3>
    componentLinearVelScale=<0.3 0.3 0.3>
    gravityOverride {z=-980}
    physicsBodyDefinitions [{
      sphereCollisionRadius=0
      boundBone {boneName=dyn_flap3}
      boxExtents=<5 5 5>
      constraintSetup {angularTargetAxis=AxisY, angularLimitsMax=<0 10 0>, angularLimitsMin=<0 0 0>, angularTarget=<0 1 0>}
      localJointOffset=<1 0 0.5>
     }]
   }
   n4 = Anim Dynamics - Bone: dyn_flap4(ComponentPose=@n3, Alpha=1, ExternalForce="0,0,0") {
    angularDampingOverride=0.98
    angularSpringConstant=20000
    bAngularSpring=true
    bGravityOverrideInSimSpace=true
    bOverrideAngularDamping=true
    bUseGravityOverride=true
    simulationSpace=RootRelative
    alphaBoolBlend {blendOption=Linear}
    boundBone {boneName=dyn_flap4}
    componentAppliedLinearAccClamp=<3000 3000 3000>
    componentLinearAccScale=<0.3 0.3 0.3>
    componentLinearVelScale=<0.3 0.3 0.3>
    gravityOverride {z=-980}
    physicsBodyDefinitions [{
      sphereCollisionRadius=0
      boundBone {boneName=dyn_flap4}
      boxExtents=<5 5 5>
      constraintSetup {angularTargetAxis=AxisY, angularLimitsMax=<0 12.5 0>, angularLimitsMin=<0 -2.5 0>, angularTarget=<0 1 0>}
      localJointOffset=<1 0 0.5>
     }]
   }
   n5 = Get No_SkyDiving() -> (No_SkyDiving:double)
   n6 = RigidBody(ComponentPose=@n4, Alpha=@n5) {
    bClampLinearTranslationLimitToRefPose=true
    bOverrideWorldGravity=true
    overridePhysicsAsset=&/F_MED_Hoodie_Bandit_Pack_Physics
    simulationSpace=BaseBoneSpace
    alphaBoolBlend {blendOption=Linear}
    baseBoneRef {boneName=pack}
    componentAppliedLinearAccClamp=<1500 1500 1500>
    componentLinearAccScale=<0.4 0.3 0.3>
    componentLinearVelScale=<0.4 0.3 0.3>
    overrideWorldGravity {z=-980}
   }
   n7 = "Transform (Modify) Bone - Bone: dyn_ribbon1_l"(ComponentPose=@n6, Alpha=@n5, Translation="0,-1,0", Rotation="0,0,0", Scale="1,1,1") {translationMode=BMM_Additive, alphaBoolBlend {blendOption=Linear}, boneToModify {boneName=dyn_ribbon1_l}, translation {y=-1}}
   n8 = "Transform (Modify) Bone - Bone: dyn_ribbon1_r"(ComponentPose=@n7, Alpha=@n5, Translation="0,-1,0", Rotation="0,0,0", Scale="1,1,1") {translationMode=BMM_Additive, alphaBoolBlend {blendOption=Linear}, boneToModify {boneName=dyn_ribbon1_r}, translation {y=-1}}
   n9 = Get SkyDyving_strops_correction() -> (SkyDyving_strops_correction:double)
   n10 = RigidBody(ComponentPose=@n8, Alpha=@n9) {
    bClampLinearTranslationLimitToRefPose=true
    bOverrideWorldGravity=true
    overridePhysicsAsset=&/F_MED_Hoodie_Bandit_Pack_Physics_SkyDiving
    simulationSpace=BaseBoneSpace
    alphaBoolBlend {blendOption=Linear}
    baseBoneRef {boneName=pack}
    componentAppliedLinearAccClamp=<1500 1500 1500>
    componentLinearAccScale=<0.4 0.3 0.3>
    componentLinearVelScale=<0.4 0.3 0.3>
   }
   n11 = "Transform (Modify) Bone - Bone: dyn_ribbon2_r"(ComponentPose=@n10, Alpha=@n9, Translation="0,0,0", Rotation="0,0,30", Scale="1,1,1") {rotationMode=BMM_Additive, alphaBoolBlend {blendOption=Linear}, boneToModify {boneName=dyn_ribbon2_r}, rotation {roll=30}}
   n12 = "Transform (Modify) Bone - Bone: dyn_ribbon2_l"(ComponentPose=@n11, Alpha=@n9, Translation="0,0,0", Rotation="0,0,30", Scale="1,1,1") {rotationMode=BMM_Additive, alphaBoolBlend {blendOption=Linear}, boneToModify {boneName=dyn_ribbon2_l}, rotation {roll=30}}
   Output Pose(Result=Component To Local(ComponentPose=@n12)) {layerGroup=None}
  ]}
 EventGraph {nodes [
   Event BlueprintUpdateAnimation()
   n1 = Get bIsSkyDiving()
   > Set SkyDyving_strops_correction(SkyDyving_strops_correction:double=SelectFloat(A=1, B=0, bPickA=@n1)) -> (Output_Get:double)
   > Set No_SkyDiving(No_SkyDiving:double=SelectFloat(A=0, B=1, bPickA=@n1)) -> (Output_Get:double)
  ]}
]
variables [
 No_SkyDiving {type=number}
 SkyDyving_strops_correction {type=number}
]
